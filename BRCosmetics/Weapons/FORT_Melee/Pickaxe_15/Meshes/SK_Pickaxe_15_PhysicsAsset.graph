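# Physics asset for the Pickaxe 15 skeletal mesh, a two-handed melee cosmetic, providing its simplified collision bodies.
# A narrower capsule sits on the Melee_TwoHanded_Handle bone and a larger, wider one on b_AR_Root.
# A single constraint joins the two bodies, with the same angle set for both swing limits and the twist limit.
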PhysicsAsset /BRCosmetics/Weapons/FORT_Melee/Pickaxe_15/Meshes/SK_Pickaxe_15_PhysicsAsset
bodies {
 Melee_TwoHanded_Handle [{shapeType=Capsule, length=29.6981, radius=7.3734, center=<3.3238 -1.3215 -14.2164>, rotation=<rot 0.8118 -90 -3.3015>}]
 b_AR_Root [{shapeType=Capsule, length=36.1578, radius=21.8458, center=<19.1476 -2.171 18.0852>, rotation=<rot -0.3005 -90 22.7565>}]
}
constraints [{bone1Name=Melee_TwoHanded_Handle, bone2Name=b_AR_Root, swing1LimitDegrees=45, swing2LimitDegrees=45, twistLimitDegrees=45}]
